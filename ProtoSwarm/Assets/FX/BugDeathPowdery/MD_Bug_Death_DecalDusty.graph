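# Translucent deferred decal material for the dusty impact mark of a bug death effect.
# A grayscale dirt texture, sampled with rotated, clamped UVs, is contrast-shaped into a mask.
# Decal color alpha drives a power-curved threshold subtracted from that mask, so the mark erodes as it fades, rather than only dimming.
# Decal color RGB supplies the tint, and a slightly flattened impact normal map adds surface detail.
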Material /ProtoSwarm/Assets/FX/BugDeathPowdery/MD_Bug_Death_DecalDusty
graph {
 properties {bAutomaticallySetUsageInEditor=false, blendMode=BLEND_Translucent, materialDomain=MD_DeferredDecal, usedShadingModels=Default Lit}
 nodes [
  MaterialFunctionCall(
   Normal=TextureSampleParameter2D() {parameterName=Tex2, samplerType=SAMPLERTYPE_Normal, texture=&/T_BugSplat__NormalImpact}.RGB
   Flatness=Constant() {r=0.22}
  ) {materialFunction=&FlattenNormal}
  n1 = DecalColor()
  Multiply(A=@n1.RGB) {constA=1}
  n2 = TextureCoordinate()
  n3 = Constant() {r=0.5}
  n4 = Constant() {r=0.33}
  n5 = TextureSampleParameter2D(UVs=Clamp(None=MaterialFunctionCall(UVs=@n2, Rotation Center=@n3, "Rotation Angle (0-1)"=@n4) {materialFunction=&CustomRotator}.Rotated Values)) {parameterName=Tex1, samplerType=SAMPLERTYPE_LinearGrayscale, texture=&/T_BugSplat__DirtImpact}
  n6 = Multiply(A=@n5.R) {constB=1.5}
  n7 = Saturate(None=LinearInterpolate(A=@n5.R, B=Power(Base=@n6) {constExponent=1.5}))
  n8 = LinearInterpolate(Alpha=Power(Base=Saturate(None=@n1.A)) {constExponent=0.3}) {constA=1, constB=-1}
  Multiply(A=@n1.A, B=Multiply(A=@n7, B=Subtract(A=@n7, B=@n8))) {constA=1}
 ]
}
